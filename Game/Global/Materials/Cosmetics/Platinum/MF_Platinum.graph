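# Material function for a platinum cosmetic finish, blending an anisotropic metallic layer into the incoming material through a tech-art channel mask or full override.
# Base color combines time-of-day tint, gradient lookup, thin-film iridescence from a view-angle spectrum lookup, and scrolling, noise-warped glints mapped in pre-skinned bounds space.
# A pattern mask darkens the color, and a static switch adds extra platinum masking.
# Emissive scales highlight and iridescence by audio amplitude above a base floor.
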MaterialFunction /Game/Global/Materials/Cosmetics/Platinum/MF_Platinum
graph {nodes [
  n1 = TextureSampleParameter2D(UVs=Multiply(A=TextureCoordinate(), B=ScalarParameter() {defaultValue=4, parameterName=IridescentHighlightUVWarpTiling})) {parameterName=ThinFilmUVOffsetTexture, texture=/Game/Global/Textures/Default/Noise/3xPerlin_M}
  ChannelMaskParameter(None=@n1.RGB) {group=06_Thin Film, maskChannel=Green, parameterName=ThinFilmUVOffsetChannel, defaultValue=<rgba 0 1 0 0>}
  FunctionOutput(None=BlendMaterialAttributes(A=NamedRerouteUsage(), B=NamedRerouteUsage(), Alpha=NamedRerouteUsage()))
  n2 = NamedRerouteUsage()
  n3 = Multiply(A=@n2, B=ScalarParameter() {defaultValue=1, group=Platinum, parameterName=PlatinumMaskGlintStrength})
  n4 = NamedRerouteUsage()
  n5 = VectorParameter() {parameterName=Color_Morning, defaultValue=<rgba 0.25 0.25 0.25 1>}
  n6 = VectorParameter() {parameterName=Color_Day, defaultValue=<rgba 0.125 0.125 0.125 1>}
  n7 = VectorParameter() {parameterName=Color_Evening, defaultValue=<rgba 0.25 0.25 0.25 1>}
  n8 = VectorParameter() {parameterName=Color_Night, defaultValue=<rgba 0.25 0.25 0.25 1>}
  n9 = NamedRerouteUsage()
  n10 = TextureSampleParameter2D() {group=Platinum, parameterName=PlatinumGradientMask, texture=&T_BlacksRGB}
  n11 = LinearInterpolate(
   A=@n9
   B=VectorParameter() {group=Platinum, parameterName=PlatinumColor, defaultValue=<rgba 0.5 0.5 0.5 1>}.RGB
   Alpha=Multiply(A=ScalarParameter() {defaultValue=1, group=Platinum, parameterName=PlatinumMaskGradientStrehgth}, B=@n10.R)
  )
  n12 = NamedRerouteUsage()
  n13 = Multiply(A=@n12, B=ScalarParameter() {defaultValue=1, group=Platinum, parameterName=PlatinumMaskIrridescenceStrength})
  n14 = NamedRerouteDeclaration(None=@n10.R) {name=PlatinumGradientMask, nodeColor=<rgba 0.7578 1 0 1>}
  n15 = Add(
   A=Add(
    A=MaterialFunctionCall(Color_Morning=@n5.RGB, Color_Day=@n6.RGB, Color_Evening=@n7.RGB, Color_Night=@n8.RGB) {materialFunction=&Color_Time_of_Day}.Result
    B=StaticSwitchParameter(True=@n11, False=@n9) {parameterName=UsePlatinumMask}
   )
   B=StaticSwitchParameter(True=LinearInterpolate(A=@n12, B=@n13, Alpha=@n14), False=@n12) {parameterName=UsePlatinumMask}
  )
  n16 = ScalarParameter() {defaultValue=0.6, parameterName=PatternColorMultiplier}
  n17 = Add(
   A=StaticSwitchParameter(True=LinearInterpolate(A=@n2, B=@n3, Alpha=@n4), False=@n2) {parameterName=UsePlatinumMask}
   B=LinearInterpolate(A=@n15, B=Multiply(A=@n15, B=@n16), Alpha=NamedRerouteUsage())
  )
  Multiply(
   A=MaterialFunctionCall(
    Value_Morning=ScalarParameter() {defaultValue=1, group=Platinum, parameterName=PlatinumBoostMorning}
    Value_Day=ScalarParameter() {defaultValue=1, group=Platinum, parameterName=PlatinumBoostDay}
    Value_Evening=ScalarParameter() {defaultValue=1, group=Platinum, parameterName=PlatinumBoostEvening}
    Value_Night=ScalarParameter() {defaultValue=1, group=Platinum, parameterName=PlatinumBoostNight}
   ) {materialFunction=&Scalar_Time_of_Day}.Result
   B=@n17
  )
  NamedRerouteDeclaration(None=FunctionInput() {inputName=Material, inputType=FunctionInput_MaterialAttributes}) {name=IN_Material, nodeColor=<rgba 0.1484 1 0 1>}
  n18 = FunctionInput() {bUsePreviewValueAsDefault=true, inputName=TechArtMask, sortPriority=1, previewValue=<1 1 1 1>}
  NamedRerouteDeclaration(None=@n18) {name=IN_TechArtMask, nodeColor=<rgba 0.1484 1 0 1>}
  n19 = Multiply(
   A=Power(Base=NamedRerouteUsage(), Exp=ScalarParameter() {defaultValue=0.8, parameterName=PlatinumUVExponent})
   B=ScalarParameter() {defaultValue=1, parameterName=PlatinumUVScale}
  )
  n20 = ScalarParameter() {defaultValue=0.4, parameterName=PlatinumUVOffset}
  n21 = TextureSample(UVs=MaterialFunctionCall(X=Add(A=@n19, B=@n20), Y=Constant() {r=0.5}) {materialFunction=&MakeFloat2}.Result) {texture=&/T_Platinum_Gradient}
  n22 = ScalarParameter() {defaultValue=1, parameterName=PlatinumStrength}
  n23 = ScalarParameter() {parameterName=PlatinumMaskThreshold}
  n24 = ScalarParameter() {defaultValue=0.1, parameterName=PlatinumMaskBias}
  n25 = Saturate(None=InverseLinearInterpolate(A=@n23, B=Add(A=@n23, B=@n24), Value=NamedRerouteUsage()))
  NamedRerouteDeclaration(None=Multiply(A=Multiply(A=@n21.RGB, B=@n22), B=Multiply(A=@n25, B=@n25))) {name=PlatinumGradient, nodeColor=<rgba 1 0 0.7734 1>}
  n26 = Normalize(VectorInput=VectorParameter() {parameterName=AnisotropicTangentDirection, defaultValue=<rgba 1 0 0 0>}.RGB)
  n27 = LinearInterpolate(A=PixelNormalWS(), B=VertexNormalWS(), Alpha=ScalarParameter() {parameterName=AnisotropicFlattenNormals})
  NamedRerouteDeclaration(None=Transform(None=CrossProduct(A=@n26, B=@n27)) {transformSourceType=TRANSFORMSOURCE_World, transformType=TRANSFORM_Tangent}) {name=Tangent, nodeColor=<rgba 0 1 0.3203 1>}
  n28 = NamedRerouteUsage()
  NamedRerouteDeclaration(None=StaticSwitchParameter(True=Constant() {r=1}, False=ChannelMaskParameter(None=@n28) {parameterName=PlatinumChannel}.R) {group=Platinum, parameterName=OverrideTechArtMask}) {name=PlatinumMask, nodeColor=<rgba 0.1719 1 0 1>}
  n29 = SetMaterialAttributes!(
   MaterialAttributes=NamedRerouteUsage()
   BaseColor=NamedRerouteUsage()
   Metallic=ScalarParameter() {defaultValue=1, parameterName=Metalness}
   Anisotropy=ScalarParameter() {defaultValue=0.75, parameterName=Anisotropy}
   Roughness=ScalarParameter() {defaultValue=0.7, parameterName=Roughness}
   Tangent=NamedRerouteUsage()
   EmissiveColor=NamedRerouteUsage()
  ) {attributeSetTypes [
    #1
    #2
    #3
    #4
    #5
    #6
   ]}
  NamedRerouteDeclaration(None=@n29) {name=PlatinumMaterial, nodeColor=<rgba 1 0.3047 0 1>}
  n30 = NamedRerouteUsage()
  n31 = Multiply(A=ComponentMask(None=@n1.RGB) {g=true, r=true}) {constB=0.1}
  n32 = ScalarParameter() {defaultValue=0.5, parameterName=IridescentHighlightUVWarpStrength}
  n33 = Multiply(
   A=Time()
   B=ComponentMask(None=VectorParameter() {parameterName=IridescentHighlightSpeed, defaultValue=<rgba 0 0.2 0 1>}.RGB) {g=true, r=true}
  )
  n34 = LocalPosition() {localOrigin=InstancePreSkinning}
  n35 = PreSkinnedLocalBounds()
  n36 = Subtract(A=VertexInterpolator(VS=ComponentMask(None=@n34.XYZ) {b=true, r=true}).PS, B=ComponentMask(None=@n35.Min) {b=true, r=true})
  n37 = ComponentMask(None=@n35.Extents) {b=true, r=true}
  n38 = Rotator(
   Coordinate=Multiply(A=Divide(A=@n36, B=@n37), B=ScalarParameter() {defaultValue=1, parameterName=IridescentHighlightTile})
   Time=ScalarParameter() {defaultValue=0.5, parameterName=IridescentHighlightRotation}
  )
  n39 = TextureSample(UVs=Add(A=Multiply(A=@n31, B=@n32), B=Add(A=@n33, B=@n38))) {samplerType=SAMPLERTYPE_LinearColor, texture=&/T_Platinum_Noise}
  n40 = ScalarParameter() {defaultValue=3, parameterName=IridescentHighlightStrength}
  n41 = ScalarParameter() {defaultValue=0.1, parameterName=IridescentHighlightBaseStrength}
  n42 = ScalarParameter() {defaultValue=1, parameterName=IridescentHighlightPatternStrength}
  n43 = NamedRerouteUsage()
  n44 = Multiply(A=Multiply(A=Multiply(A=@n39.RGB, B=@n40), B=LinearInterpolate(A=@n41, B=@n42, Alpha=@n43)), B=NamedRerouteUsage())
  NamedRerouteDeclaration(None=Desaturation(None=Multiply(A=@n30, B=@n44), Fraction=ScalarParameter() {parameterName=IridescentHighlightDesaturation})) {name=Glint, nodeColor=<rgba 1 0.6094 0 1>}
  n45 = ScalarParameter() {defaultValue=0.2, parameterName=IridescentHighlightMaskThreshold}
  n46 = Add(A=@n45, B=ScalarParameter() {defaultValue=0.4, parameterName=IridescentHighlightMaskBias})
  n47 = NamedRerouteUsage()
  NamedRerouteDeclaration(None=Saturate(None=InverseLinearInterpolate(A=@n45, B=@n46, Value=@n47))) {name=HighlightMask, nodeColor=<rgba 0.1484 0 1 1>}
  NamedRerouteDeclaration(None=@n17) {name=BaseColor, nodeColor=<rgba 0.125 0 1 1>}
  NamedRerouteDeclaration(None=ChannelMaskParameter(None=@n28) {maskChannel=Green, parameterName=PatternChannel, defaultValue=<rgba 0 1 0 0>}.R) {name=PatternMask, nodeColor=<rgba 0 1 0.9062 1>}
  n48 = NamedRerouteDeclaration(None=MaterialFunctionCall() {materialFunction=/Game/Packages/Fortress_SharedMaterials/Base_Material_Functions/BaseCharacterShaderFunctions/Audio/MF_AudioReactivity}.Amplitude) {name=Amplitude, nodeColor=<rgba 0 0.8125 1 1>}
  n49 = ScalarParameter() {defaultValue=3, parameterName=AudioReactivityExponent}
  n50 = Max(
   A=Multiply(A=Power(Base=@n48, Exp=@n49), B=ScalarParameter() {defaultValue=1.35, parameterName=FrontEndReactivityBoost})
   B=ScalarParameter() {defaultValue=0.2, parameterName=BaseEmissiveAmount}
  )
  n51 = NamedRerouteUsage()
  n52 = ScalarParameter() {defaultValue=0.5, parameterName=HighlightReactiveStrength}
  n53 = NamedRerouteUsage()
  n54 = ScalarParameter() {defaultValue=0.5, parameterName=IridescenceReactiveStrength}
  n55 = Desaturation(
   None=Add(A=Multiply(A=@n51, B=@n52), B=Multiply(A=@n53, B=@n54))
   Fraction=ScalarParameter() {defaultValue=0.3, parameterName=HighlightReactiveDesaturation}
  )
  n56 = ScalarParameter() {defaultValue=1, parameterName=PlatinumReactivity}
  NamedRerouteDeclaration(None=Multiply(A=Multiply(A=@n50, B=@n55), B=Saturate(None=@n56))) {name=Emissive, nodeColor=<rgba 1 0.8672 0 1>}
  n57 = DotProduct(A=CameraVectorWS(), B=PixelNormalWS())
  n58 = NamedRerouteDeclaration(None=Divide(A=ArccosineFast!(None=@n57)) {constB=1.5707}) {name=LinearFresnel, nodeColor=<rgba 0.4531 0 1 1>}
  n59 = ScalarParameter() {defaultValue=0.624, parameterName=IridesenceUVExponent}
  n60 = Add(
   A=Multiply(A=Power(Base=@n58, Exp=@n59), B=ScalarParameter() {defaultValue=2.6501, parameterName=IridesenceUVScale})
   B=ScalarParameter() {defaultValue=1.3275, parameterName=IridesenceUVOffset}
  )
  n61 = Constant() {r=0.5}
  n62 = NamedRerouteDeclaration(None=TextureSample(UVs=MaterialFunctionCall(X=@n60, Y=@n61) {materialFunction=&MakeFloat2}.Result) {texture=/Game/Global/Textures/Default/Thinfilm/T_ThinFilm_Spectrum_BlueGreen_Color}.RGB) {name=IridesenceColor, nodeColor=<rgba 0.1953 0 1 1>}
  n63 = ScalarParameter() {defaultValue=0.3, parameterName=IridesenceStrength}
  n64 = ScalarParameter() {defaultValue=0.2, parameterName=IridesenceMaskThreshold}
  n65 = ScalarParameter() {defaultValue=0.125, parameterName=IridesenceMaskBias}
  n66 = Saturate(None=InverseLinearInterpolate(A=@n64, B=Add(A=@n64, B=@n65), Value=NamedRerouteUsage()))
  NamedRerouteDeclaration(None=Multiply(A=Multiply(A=@n62, B=@n63), B=Multiply(A=@n66, B=@n66))) {name=Iridesence, nodeColor=<rgba 1 0 0.9141 1>}
  NamedRerouteUsage()
 ]}
